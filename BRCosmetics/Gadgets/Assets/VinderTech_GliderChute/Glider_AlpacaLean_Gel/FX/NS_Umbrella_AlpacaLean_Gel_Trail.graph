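# Niagara trail effect for the AlpacaLean Gel glider, firing muzzle flashes, smoke and steam from skeletal mesh sockets as it moves.
# Forward and side muzzle emitters burst on ForwardDot and RightDot thresholds, stepping through twelve shuffled sockets, with audio players attached.
# Smoke sprites spawn from muzzle particles via attribute readers; steam meshes use a second socket set.
# Front-end preview force-fires forward muzzles and hides steam; time of day scales smoke and steam alpha.
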NiagaraSystem /BRCosmetics/Gadgets/Assets/VinderTech_GliderChute/Glider_AlpacaLean_Gel/FX/NS_Umbrella_AlpacaLean_Gel_Trail
dependencies {
 usedDataInterfaces [
  &NiagaraDataInterfaceAudioPlayer
  &NiagaraDataInterfaceArrayFloat
  &NiagaraDataInterfaceArrayInt32
  &NiagaraDataInterfaceArrayFloat2
  &NiagaraDataInterfaceVectorCurve
  &NiagaraDataInterfaceCurve
  &NiagaraDataInterfaceParticleRead
  &NiagaraDataInterfaceSkeletalMesh
 ]
 usedDynamicInputs []
 usedModules [
  &self:SystemSpawnScript.SystemScriptSource.SystemScriptGraph.NiagaraNodeAssignment_1.SetVariables_#1
  &SystemState
  &Timeline
 ]
}
emitters {
 NE_Muzzle_Forward {
  emitterSpawnScript [&InitializeAudioPlayer()]
  emitterUpdateScript [
   /Niagara/Modules/Audio/PlayPersistentAudio()
   /Niagara/Modules/Audio/UpdatePersistentAudio(Update Type:ENiagara_AudioParamType=::NewEnumerator5)
   &EmitterState()
   /Game/Effects/Niagara/Modules/Spawn/FN_Spawn_TriggeredBurst(
    Cooldown:NiagaraFloat=&MakeCustomFloatFromBool(Bool=&SetBoolByFloatComparison(A=$System.Timeline001.Time, B=0), True Float=0.1, False Float=1)
    Module.Use Burst Count Limit=false
    Spawn:NiagaraBool=&Boolean_OrOperation(
     Bool1=&SetBoolByFloatComparison(A=$User.ForwardDot, B=0.5)
     Bool2=&SetBoolByFloatComparison(A=$User.ForwardDot, B=-0.5, Comparison Type=::NewEnumerator6)
    )
    Spawn Group Mode=2
    Spawn Group Reset Count=12
    UseLoopCountLimit=false
   )
   &FN_Spawn_TriggeredBurst(
    Cooldown=0.2
    Module.Use Burst Count Limit=false
    Spawn:NiagaraBool=$User.bIsFrontEndPreview
    Spawn Group Mode=2
    Spawn Group Reset Count=12
    UseLoopCountLimit=false
   ) {module=FN_Spawn_TriggeredBurst001}
  ]
  particleSpawnScript [
   &InitializeParticle(
    Lifetime=&RandomRangeFloat(Minimum=0.125, Maximum=0.155)
    Mesh Scale=<0.75 0.5 0.5>
    Mesh Scale Mode=::NewEnumerator4
    UsePositionOffset=false
    Write Material Random=true
   )
   &InitialMeshOrientation(Mesh Orientation Mode=::NewEnumerator2, Orientation Vector=<0 0 -1>, Use Rotation Vector=false)
   &PartitionParticles(Number Of Partitions=12)
   &self:ScratchModule(Audio Player:NiagaraDataInterfaceAudioPlayer=$Emitter.Audio Player) {module=ScratchModule}
  ]
  particleUpdateScript [
   &ParticleState()
   &SkeletalMeshLocation(
    Bone / Socket Index=&MakeCustomIntFromBool(
     Bool=$User.bIsFrontEndPreview
     True INT=&SelectIntFromArray(
      Direct Array Index=$Particles.PartitionParticles.Partition
      Int Selection Array=DataInterface() {propertyValues {IntData [0,11,7,3,10,5,9,2,8,6,4,1]}}
     )
     False INT=&SelectIntFromArray(
      Direct Array Index=$Emitter.FN_Spawn_TriggeredBurst.Current Spawn Group
      Int Selection Array=DataInterface() {propertyValues {IntData [0,11,7,3,10,5,9,2,8,6,4,1]}}
     )
    )
    MaskSpawnGroup=false
    Mesh Sampling Type=::NewEnumerator5
    Skeletal Mesh=$System.Skeletal Mesh
    Socket Sampling Mode=::NewEnumerator5
   )
   &ScaleMeshSize(Scale Factor=&VectorFromCurve(VectorCurve=DataInterface() {propertyValues {XCurve {keys [<at 0 0>,<at 1 1>]}, YCurve {keys [<at 0 0.25>,<at 1 1.25>]}, ZCurve {keys [<at 0 0.25>,<at 1 1.25>]}}}))
   &ScaleColor(
    Scale Alpha=&FloatFromCurve(
     FloatCurve=DataInterface() {propertyValues {Curve {keys [
         {interpMode=RCIM_Cubic, tangentMode=RCTM_User, time=0, value=0}
         {interpMode=RCIM_Cubic, tangentMode=RCTM_User, time=0.1, value=1}
         {interpMode=RCIM_Cubic, tangentMode=RCTM_User, time=1, value=0}
        ]}}}
     Scale Curve=&RandomRangeFloat(Minimum=0.5, Maximum=1)
    )
    Scale RGB=&VectorFromFloat(Value=&FloatFromCurve(FloatCurve=DataInterface() {propertyValues {Curve {keys [{arriveTangent=-42.9916, interpMode=RCIM_Cubic, leaveTangent=-42.9916, tangentMode=RCTM_User, time=0, value=5}, <at 0.15 1 RCIM_Cubic>]}}}))
    ScaleA=true
    ScaleRGB=true
   )
   &DynamicMaterialParameters(
    Index 0 Param 1=&FloatFromCurve(
     FloatCurve=DataInterface() {propertyValues {Curve {keys [
         {arriveTangent=2.9073, interpMode=RCIM_Cubic, leaveTangent=2.9073, leaveTangentWeight=0.3077, tangentMode=RCTM_User, time=0.35, value=0}
         {arriveTangentWeight=0.8354, interpMode=RCIM_Cubic, tangentMode=RCTM_User, time=1, value=1}
        ]}}}
     Scale Curve=0.75
    )
    Index 0 Param 2=&FloatFromCurve(FloatCurve=DataInterface() {propertyValues {Curve {keys [
         {interpMode=RCIM_Cubic, leaveTangentWeight=0.3627, tangentMode=RCTM_User, time=0.4, value=1}
         {arriveTangent=-1.799, arriveTangentWeight=1.0345, interpMode=RCIM_Cubic, leaveTangent=-1.799, tangentMode=RCTM_User, time=1, value=0}
        ]}}})
    Index 0 Param 3=&MakeCustomFloatFromBool(Bool=&RandomBool(), True Float=1, False Float=0)
    Module.Index0Param0Write=true
    Module.Index0Param1Write=true
    Module.Index0Param2Write=true
    Module.Index0Param3Write=false
   )
  ]
  rendererClasses [&NiagaraMeshRendererProperties]
  renderers [{rendererClass=&NiagaraMeshRendererProperties, rendererIndex=0}]
 }
 NE_Muzzle_Side {
  emitterSpawnScript [&InitializeAudioPlayer()]
  emitterUpdateScript [
   &EmitterState()
   &FN_Spawn_TriggeredBurst(
    Cooldown:NiagaraFloat=&MakeCustomFloatFromBool(Bool=&SetBoolByFloatComparison(A=$System.Timeline.Time, B=0), True Float=0.5, False Float=0.5)
    Module.Use Burst Count Limit=false
    Spawn:NiagaraBool=&Boolean_OrOperation(
     Bool1=&SetBoolByFloatComparison(A=$User.RightDot, B=0.5)
     Bool2=&SetBoolByFloatComparison(A=$User.RightDot, B=-0.5, Comparison Type=::NewEnumerator6)
    )
    Spawn Group Mode=2
    Spawn Group Reset Count=12
    UseLoopCountLimit=false
   )
   &PlayPersistentAudio()
   &UpdatePersistentAudio(Update Type:ENiagara_AudioParamType=::NewEnumerator5)
  ]
  particleSpawnScript [
   &InitializeParticle(
    Lifetime=&RandomRangeFloat(Minimum=0.125, Maximum=0.155)
    Mesh Scale=<0.75 0.5 0.5>
    Mesh Scale Mode=::NewEnumerator4
    UsePositionOffset=false
    Write Material Random=true
   )
   &InitialMeshOrientation(Mesh Orientation Mode=::NewEnumerator2, Orientation Vector=<0 0 -1>, Use Rotation Vector=false)
   &self:ScratchModule(Audio Player:NiagaraDataInterfaceAudioPlayer=$Emitter.Audio Player) {module=ScratchModule}
  ]
  particleUpdateScript [
   &ParticleState()
   &SkeletalMeshLocation(
    Bone / Socket Index=&SelectIntFromArray(
     Direct Array Index=$Emitter.FN_Spawn_TriggeredBurst.Current Spawn Group
     Int Selection Array=DataInterface() {propertyValues {IntData [0,11,7,3,10,5,9,2,8,6,4,1]}}
    )
    MaskSpawnGroup=false
    Mesh Sampling Type=::NewEnumerator5
    Skeletal Mesh=$System.Skeletal Mesh
    Socket Sampling Mode=::NewEnumerator5
   )
   &ScaleMeshSize(Scale Factor=&VectorFromCurve(VectorCurve=DataInterface() {propertyValues {XCurve {keys [<at 0 0>,<at 1 1>]}, YCurve {keys [<at 0 0.25>,<at 1 1.25>]}, ZCurve {keys [<at 0 0.25>,<at 1 1.25>]}}}))
   &ScaleColor(
    Scale Alpha=&FloatFromCurve(
     FloatCurve=DataInterface() {propertyValues {Curve {keys [
         {interpMode=RCIM_Cubic, tangentMode=RCTM_User, time=0, value=0}
         {interpMode=RCIM_Cubic, tangentMode=RCTM_User, time=0.1, value=1}
         {interpMode=RCIM_Cubic, tangentMode=RCTM_User, time=1, value=0}
        ]}}}
     Scale Curve=&RandomRangeFloat(Minimum=0.5, Maximum=1)
    )
    Scale RGB=&VectorFromFloat(Value=&FloatFromCurve(FloatCurve=DataInterface() {propertyValues {Curve {keys [{arriveTangent=-42.9916, interpMode=RCIM_Cubic, leaveTangent=-42.9916, tangentMode=RCTM_User, time=0, value=5}, <at 0.15 1 RCIM_Cubic>]}}}))
    ScaleA=true
    ScaleRGB=true
   )
   &DynamicMaterialParameters(
    Index 0 Param 1=&FloatFromCurve(
     FloatCurve=DataInterface() {propertyValues {Curve {keys [
         {arriveTangent=2.9073, interpMode=RCIM_Cubic, leaveTangent=2.9073, leaveTangentWeight=0.3077, tangentMode=RCTM_User, time=0.35, value=0}
         {arriveTangentWeight=0.8354, interpMode=RCIM_Cubic, tangentMode=RCTM_User, time=1, value=1}
        ]}}}
     Scale Curve=0.75
    )
    Index 0 Param 2=&FloatFromCurve(FloatCurve=DataInterface() {propertyValues {Curve {keys [
         {interpMode=RCIM_Cubic, leaveTangentWeight=0.3627, tangentMode=RCTM_User, time=0.4, value=1}
         {arriveTangent=-1.799, arriveTangentWeight=1.0345, interpMode=RCIM_Cubic, leaveTangent=-1.799, tangentMode=RCTM_User, time=1, value=0}
        ]}}})
    Index 0 Param 3=&MakeCustomFloatFromBool(Bool=&RandomBool(), True Float=1, False Float=0)
    Module.Index0Param0Write=true
    Module.Index0Param1Write=true
    Module.Index0Param2Write=true
    Module.Index0Param3Write=false
   )
  ]
  rendererClasses [&NiagaraMeshRendererProperties]
  renderers [{rendererClass=&NiagaraMeshRendererProperties, rendererIndex=0}]
 }
 NE_Smoke_Forward {
  emitterUpdateScript [
   &EmitterState(Life Cycle Mode=::NewEnumerator1, Loop Behavior=::NewEnumerator1, Loop Duration Mode=::NewEnumerator1, UseLoopDelay=false)
   &SpawnParticlesFromOtherEmitter(
    Attribute Reader=DataInterface() {propertyValues {EmitterBinding {bindingMode=Other, emitterName=NE_Muzzle_Forward}}}
    Spawn Rate=20
    Use Maximum Age=false
    Use Minimum Age=false
   )
  ]
  particleSpawnScript [
   &InitializeParticle(
    Lifetime=0.3
    Sprite Size Max=<30 60>
    Sprite Size Min=<25 50>
    Sprite Size Mode=::NewEnumerator2
    UsePositionOffset=false
    Write Material Random=true
   )
   &SampleParticlesFromOtherEmitter(Mesh Orientation Sampling:ENiagara_DirectReadSamplingMode=::NewEnumerator0)
   &ApplyInitialForces()
   &AlignSpriteToMeshOrientation(
    Mesh Orientation Relative Sprite Alignment Vector=<1 0 0>
    Mesh Orientation Relative Sprite Facing Vector:Vector3f=<0 0 -1>
    SetAlignment=true
    SetFacing=true
   )
  ]
  particleUpdateScript [
   &ParticleState()
   &SolveForcesAndVelocity(Clamp Velocity=false, Limit Acceleration=false)
   /SetVariables(Particles.DynamicParameterAlt1:Vector4f=&MakeVector4(
     X=&FloatFromCurve(FloatCurve=DataInterface() {propertyValues {Curve {keys [<at 0 0>,<at 1 1>]}}})
     Y=&UniformAOrBFloat(A=0, Evaluation Type=::NewEnumerator0, Recalculate Random Each Loop=true, Override Seed=false)
     Z=0
     W=0
    )) {isSetParameters=true, module=SetVariables_#2}
   &ScaleColor(
    Scale Alpha=&FNTimeOfDayScalar(Night Scale=0.5, Morning Scale=1, Evening Scale=1)
    ScaleA=true
    ScaleRGB=true
   )
  ]
  rendererClasses [&NiagaraSpriteRendererProperties]
  renderers [{rendererClass=&NiagaraSpriteRendererProperties}]
 }
 NE_Smoke_Side {
  emitterUpdateScript [
   &EmitterState(Life Cycle Mode=::NewEnumerator1, Loop Behavior=::NewEnumerator1, Loop Duration Mode=::NewEnumerator1, UseLoopDelay=false)
   &SpawnParticlesFromOtherEmitter(
    Attribute Reader=DataInterface() {propertyValues {EmitterBinding {bindingMode=Other, emitterName=NE_Muzzle_Side}}}
    Spawn Rate=20
    Use Maximum Age=false
    Use Minimum Age=false
   )
  ]
  particleSpawnScript [
   &InitializeParticle(
    Lifetime=0.3
    Sprite Size Max=<30 60>
    Sprite Size Min=<25 50>
    Sprite Size Mode=::NewEnumerator2
    UsePositionOffset=false
    Write Material Random=true
   )
   &SampleParticlesFromOtherEmitter(Mesh Orientation Sampling:ENiagara_DirectReadSamplingMode=::NewEnumerator0)
   &ApplyInitialForces()
   &AlignSpriteToMeshOrientation(
    Mesh Orientation Relative Sprite Alignment Vector=<1 0 0>
    Mesh Orientation Relative Sprite Facing Vector:Vector3f=<0 0 -1>
    SetAlignment=true
    SetFacing=true
   )
  ]
  particleUpdateScript [
   &ParticleState()
   &SolveForcesAndVelocity(Clamp Velocity=false, Limit Acceleration=false)
   /SetVariables(Particles.DynamicParameterAlt1:Vector4f=&MakeVector4(
     X=&FloatFromCurve(FloatCurve=DataInterface() {propertyValues {Curve {keys [<at 0 0>,<at 1 1>]}}})
     Y=&UniformAOrBFloat(A=0, Evaluation Type=::NewEnumerator0, Recalculate Random Each Loop=true, Override Seed=false)
     Z=0
     W=0
    )) {isSetParameters=true, module=SetVariables_#2}
   &ScaleColor(
    Scale Alpha=&FNTimeOfDayScalar(Night Scale=0.5, Morning Scale=1, Evening Scale=1)
    ScaleA=true
    ScaleRGB=true
   )
  ]
  rendererClasses [&NiagaraSpriteRendererProperties]
  renderers [{rendererClass=&NiagaraSpriteRendererProperties}]
 }
 NE_Steam {
  emitterUpdateScript [
   &EmitterState()
   &FN_Spawn_TriggeredBurst(
    Cooldown=0.1
    Module.Use Burst Count Limit=false
    Spawn:NiagaraBool=&SetBoolByFloatComparison(A=$User.ForwardDot, B=0, Comparison Type=::NewEnumerator4)
    UseLoopCountLimit=false
   )
  ]
  particleSpawnScript [
   &InitializeParticle(
    Color=<rgba 1 1 1 0.35>
    Lifetime=&RandomRangeFloat(Minimum=1, Maximum=2)
    Mesh Scale Max=<1 1 3>
    Mesh Scale Min=<1 1 0.5>
    Mesh Scale Mode=::NewEnumerator2
    UsePositionOffset=false
    Write Material Random=true
   )
   &InitialMeshOrientation(Mesh Axis To Orient=<0 0 0>, Mesh Orientation Mode=::NewEnumerator2, Use Rotation Vector=false)
   &AddVelocity(Velocity=<0 0 0>)
  ]
  particleUpdateScript [
   &ParticleState()
   &SkeletalMeshLocation(
    Bone / Socket Index=&SelectIntFromArray(
     Array Sampling Mode=::NewEnumerator1
     Direct Array Index=?
     Int Selection Array=DataInterface() {propertyValues {IntData [12,13,14,15,16,17,18,19,20,21,22,23]}}
     Random Seed=0
    )
    MaskSpawnGroup=false
    Mesh Sampling Type=::NewEnumerator5
    Orientation Sampling=::NewEnumerator1
    Skeletal Mesh=$System.Skeletal Mesh
    Socket Sampling Mode=::NewEnumerator5
   )
   &DynamicMaterialParameters(
    Index 0 Param 1=&RandomRangeFloat(Minimum=0.5, Maximum=1)
    Index 0 Param 2=&RandomRangeFloat(Minimum=0.8, Maximum=1)
    Index 0 Param 3=&RandomRangeFloat(Minimum=0.5, Maximum=1)
    Index 0 Param 4=&RandomRangeFloat(Minimum=0.5, Maximum=1)
    Index 3 Param 1=&RandomRangeFloat(Minimum=5, Maximum=10)
    Index 3 Param 2=&RandomRangeFloat(Minimum=15, Maximum=25)
    Index 3 Param 3=&RandomRangeFloat(Minimum=3, Maximum=5)
    Index 3 Param 4=&RandomRangeFloat(Minimum=0.5, Maximum=1)
    Module.Index0Float0Write=true
    Module.Index0Float1Write=true
    Module.Index0Float2Write=true
    Module.Index0Float3Write=true
    Module.Index3Float0Write=true
    Module.Index3Float1Write=true
    Module.Index3Float2Write=true
    Module.Index3Float3Write=true
    Write Parameter Index 3=::NewEnumerator1
   )
   &ScaleColor(
    Scale Alpha=&FloatFromCurve(
     FloatCurve=DataInterface() {propertyValues {Curve {keys [
         {interpMode=RCIM_Cubic, tangentMode=RCTM_User, time=0, value=0}
         {arriveTangent=2, interpMode=RCIM_Cubic, leaveTangent=0.0753, tangentMode=RCTM_Break, time=0.25, value=1}
         {interpMode=RCIM_Cubic, tangentMode=RCTM_User, time=0.5, value=1}
         {interpMode=RCIM_Cubic, tangentMode=RCTM_User, time=1, value=0}
        ]}}}
     Scale Curve=&Lerp_Float(
      A=&FNTimeOfDayScalar(Night Scale=0.75, Morning Scale=1, Evening Scale=1)
      B=0
      Alpha=&MakeCustomFloatFromBool(Bool=$User.bIsFrontEndPreview, True Float=1, False Float=0)
     )
    )
    ScaleA=true
    ScaleRGB=true
   )
   &SolveForcesAndVelocity(Clamp Velocity=false, Limit Acceleration=false)
  ]
  rendererClasses [&NiagaraMeshRendererProperties]
  renderers [{rendererClass=&NiagaraMeshRendererProperties, rendererIndex=0}]
 }
}
summary {
 systemName=NS_Umbrella_AlpacaLean_Gel_Trail
 emitters [
  {emitterName=NE_Muzzle_Forward, rendererClasses [&NiagaraMeshRendererProperties]}
  {emitterName=NE_Steam, rendererClasses [&NiagaraMeshRendererProperties]}
  {emitterName=NE_Smoke_Forward, rendererClasses [&NiagaraSpriteRendererProperties]}
  {emitterName=NE_Muzzle_Side, rendererClasses [&NiagaraMeshRendererProperties]}
  {emitterName=NE_Smoke_Side, rendererClasses [&NiagaraSpriteRendererProperties]}
 ]
 userVariables [
  User.bIsFrontEnd {type=NiagaraBool}
  User.bIsFrontEndPreview {type=NiagaraBool}
  User.ForwardDot {type=NiagaraFloat, defaultValue=0, description=""}
  User.RightDot {type=NiagaraFloat, defaultValue=0, description=""}
 ]
}
